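# Player animation layer Blueprint that poses a character riding a wolf.
# An event copies locomotion state from the wolf's Grandma_AnimBP to drive a full-body state machine covering idle, moving, stops, turns, jumps, slope sliding and swimming.
# Upper-body and incoming poses blend per bone, with ride mode switching between slope/head aim offsets and body-yaw/combat additives, then inertialization.
# A final-pose layer adds petting offset, root translation and counter-rotation.
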AnimBlueprint /RidingContent/Animation/HardSaddle/Wolf/WolfRidingPlayerLayerAnimBP
target_skeleton=&Fortnite_M_Avg_Player_Skeleton
graphs [
 AnimGraph {nodes [Output Pose() {layerGroup=None}]}
 EventGraph {nodes [
   n1 = Event BlueprintCollectRidableAnimBPData!() -> (OutputDelegate, then, RidableAnimInstance:Anim Instance Object Reference) {is=title K2Node_Event}
   > GetMainAnimBPData_Wolf!(execute, self, AnimInstance:Object Reference=@n1.RidableAnimInstance) -> (then)
   Event UpdateBoostPlayRate() {is=title K2Node_Event}
   > Set BoostPlayRate(BoostPlayRate:float=MapRangeClamped(Value=Get WolfSpeed2D() -> (WolfSpeed2D:double), InRangeA=0, InRangeB=1000, OutRangeA=0.4)) -> (Output_Get:float)
   Event PlaceHandIKTargets() {is=title K2Node_Event}
  ]}
 GetMainAnimBPData_Wolf {nodes [
   n1 = GetMainAnimBPData_Wolf!() -> (then, AnimInstance:Object Reference) {is=title K2Node_FunctionEntry}
   > n2 = Is Valid(InputObject=@n1)
   Return Node(execute=@n2.Is Not Valid)
   > n3 = Cast To Grandma_AnimBP!(execute, Object=@n1) -> (then, CastFailed, AsGrandma Anim BP:Grandma Anim BP Object Reference) {is=title K2Node_DynamicCast}
   > Set State Rule Idle Moving(State Rule Idle Moving:bool=Get bStateRule_Idle_Moving!(self:Fort AIFauna Anim Instance Grandma Object Reference=@n3) -> (bStateRule_Idle_Moving:bool)) -> (Output_Get:bool)
   > Set State Rule Moving Idle(State Rule Moving Idle:bool=Get bStateRule_Moving_Idle!(self:Fort AIFauna Anim Instance Grandma Object Reference=@n3) -> (bStateRule_Moving_Idle:bool)) -> (Output_Get:bool)
   > Set Is Jumping(Is Jumping:bool=Get bIsJumping!(self:Fort Anim Instance Object Reference=@n3) -> (bIsJumping)) -> (Output_Get:bool)
   > Set State Rule Fall Idle(State Rule Fall Idle:bool=Get bStateRule_Fall_Idle!(self:Fort AIFauna Anim Instance Grandma Object Reference=@n3) -> (bStateRule_Fall_Idle:bool)) -> (Output_Get:bool)
   > Set Is Falling(Is Falling:bool=Get bIsFalling!(self:Fort Anim Instance Object Reference=@n3) -> (bIsFalling)) -> (Output_Get:bool)
   > Set State Rule Fall Moving(State Rule Fall Moving:bool=Get bStateRule_Fall_Moving!(self:Fort AIFauna Anim Instance Grandma Object Reference=@n3) -> (bStateRule_Fall_Moving:bool)) -> (Output_Get:bool)
   > Set State Rule Jump Land Moving(State Rule Jump Land Moving:bool=Get bStateRule_JumpLand_Moving!(self:Fort AIFauna Anim Instance Grandma Object Reference=@n3) -> (bStateRule_JumpLand_Moving:bool)) -> (Output_Get:bool)
   > Set Is Surface Swimming(Is Surface Swimming:bool=Get bIsSurfaceSwimming!(self:Fort AIFauna Anim Instance Object Reference=@n3) -> (bIsSurfaceSwimming)) -> (Output_Get:bool)
   > Set State Rule Swim Idle Swim Loco(State Rule Swim Idle Swim Loco:bool=Get bStateRule_SwimIdle_SwimLoco!(self:Fort AIFauna Anim Instance Grandma Object Reference=@n3) -> (bStateRule_SwimIdle_SwimLoco:bool)) -> (Output_Get:bool)
   > Set BodyRoll(BodyRoll:double=Get BodyRoll!(self:Fort AIFauna Anim Instance Object Reference=@n3) -> (BodyRoll:float)) -> (Output_Get:double)
   > Set WolfSpeed2D(WolfSpeed2D:double=Get PawnSpeed2D!(self:Fort Anim Instance Object Reference=@n3) -> (PawnSpeed2D)) -> (Output_Get:double)
   > Set Is Boosting(Is Boosting:bool=Get bIsBoosting!(self:Fort AIFauna Anim Instance Object Reference=@n3) -> (bIsBoosting:bool)) -> (Output_Get:bool)
   > Set Play Land Additive(Play Land Additive:bool=Get bPlayLandAdditive!(self:Fort AIFauna Anim Instance Object Reference=@n3) -> (bPlayLandAdditive:bool)) -> (Output_Get:bool)
   > Set Is Moving Backwards(Is Moving Backwards:bool=Get bIsMovingBackwards!(self:Fort AIFauna Anim Instance Object Reference=@n3) -> (bIsMovingBackwards:bool)) -> (Output_Get:bool)
   > Set Should Play Stop Backwards(Should Play Stop Backwards:bool=Get bShouldPlayStopBackwards!(self:Fort AIFauna Anim Instance Object Reference=@n3) -> (bShouldPlayStopBackwards:bool)) -> (Output_Get:bool)
   > Set Is Turning in Place(Is Turning in Place:bool=Get bIsTurningInPlace!(self:Fort AIFauna Anim Instance Object Reference=@n3) -> (bIsTurningInPlace:bool)) -> (Output_Get:bool)
   > Set Turn in Place Angle Delta(Turn in Place Angle Delta:double=Get TurnInPlaceAngleDelta!(self:Fort AIFauna Anim Instance Object Reference=@n3) -> (TurnInPlaceAngleDelta:float)) -> (Output_Get:double)
   > Set Was Turning Right(Was Turning Right:bool=Get bWasTurningRight!(self:Fort AIFauna Anim Instance Object Reference=@n3) -> (bWasTurningRight:bool)) -> (Output_Get:bool)
   > Set Turn in Place Rotation Speed(Turn in Place Rotation Speed:double=Get TurnInPlaceRotationSpeed!(self:Fort AIFauna Anim Instance Object Reference=@n3) -> (TurnInPlaceRotationSpeed:float)) -> (Output_Get:double)
   > Set Backwards Play Rate(Backwards Play Rate:double=Get BackwardsPlayRate!(self:Fort AIFauna Anim Instance Object Reference=@n3) -> (BackwardsPlayRate:float)) -> (Output_Get:double)
   > Set TurnInPlaceModulatedSpeed(TurnInPlaceModulatedSpeed:float=Get TurnInPlaceModulatedSpeed!(self:Fort AIFauna Anim Instance Grandma Object Reference=@n3) -> (TurnInPlaceModulatedSpeed:float)) -> (Output_Get:float)
   > Set bIsTurnInPlaceAngleDeltaPositive(bIsTurnInPlaceAngleDeltaPositive:bool=Get bIsTurnInPlaceAngleDeltaPositive!(self:Fort AIFauna Anim Instance Object Reference=@n3) -> (bIsTurnInPlaceAngleDeltaPositive:bool)) -> (Output_Get:bool)
   > Set FootPhase(FootPhase:ERidingFootPhase Enum=Get FootPhase!(self:Fort AIFauna Anim Instance Object Reference=@n3) -> (FootPhase:ERidingFootPhase Enum=FeetInAir)) -> (Output_Get:ERidingFootPhase Enum=FeetInAir)
   > Set RidingMovePlayRate(RidingMovePlayRate:float=Get RidingMovePlayRate!(self:Fort AIFauna Anim Instance Object Reference=@n3) -> (RidingMovePlayRate:float)) -> (Output_Get:float)
   > Return Node()
  ]}
 VehicleFinalPoseOverride {nodes [
   n1 = Blendspace Player 'HardSaddle_Player_PettingOffset_BS'(X=0, Y=0) {blendSpace=/RidingContent/Animation/HardSaddle/Wolf/PettingOffset/HardSaddle_Player_PettingOffset_BS}
   n2 = Input Pose() {name=InFinalPose}
   n3 = "Transform (Modify) Bone - Bone: Root"!(
    ComponentPose=Local To Component(LocalPose=Apply Additive!(Base=@n2, Additive=@n1, bAlphaBoolEnabled:bool=True) -> (Pose) {alphaInputType=Bool, alphaBoolBlend {blendOption=Linear}})
    Alpha=1
    Translation=Get PlayerOffset() -> (PlayerOffset:Vector)
   ) -> (Pose) {
    rotationSpace=BCS_WorldSpace
    translationMode=BMM_Additive
    translationSpace=BCS_WorldSpace
    alphaBoolBlend {blendOption=Linear}
    boneToModify {boneName=Root}
   }
   n4 = "Transform (Modify) Bone - Bone: Root"!(
    ComponentPose=@n3
    Alpha=Get BodyCounterRotationAlpha() -> (BodyCounterRotationAlpha:float)
    Rotation:Rotator=Get BodyCounterRotation() -> (BodyCounterRotation:Rotator)
   ) -> (Pose) {
    rotationMode=BMM_Replace
    rotationSpace=BCS_WorldSpace
    translationSpace=BCS_WorldSpace
    alphaBoolBlend {blendOption=Linear}
    boneToModify {boneName=Root}
   }
   Output Pose(Result=Component To Local(ComponentPose=@n4)) {layerGroup=VehicleOverride}
  ]}
 VehicleFullBodyOverride {nodes [
   n1 = Layered blend per bone(BasePose=Use cached pose 'PostUpperBodyPose'(), BlendPoses_0=Use cached pose 'UpperAndLowerPose'()) {
    bBlendRootMotionBasedOnRootBone=false
    bMeshSpaceRotationBlend=true
    layerSetup [{branchFilters [
       {boneName=Root}
       {blendDepth=-1, boneName=pelvis}
       {boneName=Spine_01}
       {blendDepth=-1, boneName=thigh_l}
       {blendDepth=-1, boneName=thigh_r}
       {blendDepth=-1, boneName=ik_foot_root}
      ]}]
   }
   AnimGraphNode_StateMachine_2 = CombatUpperBodyAdditive() {is=title AnimGraphNode_StateMachine, bAllowConduitEntryStates=true, bSkipFirstUpdateTransition=false, maxTransitionsPerFrame=1}
   n2 = Apply Mesh Space Additive!(Base:Pose Link Structure=@n1, Additive:Pose Link Structure=@AnimGraphNode_StateMachine_2, Alpha:float=Get NoiseAlpha() -> (NoiseAlpha:float)) -> (Pose:Pose Link Structure) {is=title AnimGraphNode_ApplyMeshSpaceAdditive, alphaBoolBlend {blendInTime=0.6, blendOutTime=0.35}}
   AnimGraphNode_StateMachine_4 = EntryAdditive() {is=title AnimGraphNode_StateMachine}
   AnimGraphNode_StateMachine_3 = MeleeOffset() {is=title AnimGraphNode_StateMachine}
   n3 = Apply Additive(
    Base=AimOffset Player 'HardSaddleWolf_Player_AO_Yaw'!(
     X:float=Get BodyYawOffset() -> (BodyYawOffset:float)
     BasePose:Pose Link Structure=Apply Mesh Space Additive(Base=@n2, Additive=@AnimGraphNode_StateMachine_4, Alpha=1) {alphaBoolBlend {blendInTime=0.6, blendOutTime=0.35}}
     Alpha:float=1
    ) -> (Pose:Pose Link Structure) {blendSpace=/RidingContent/Animation/HardSaddle/Wolf/AO/HardSaddleWolf_Player_AO_Yaw, alphaBoolBlend {blendOption=Linear}}
    Additive=@AnimGraphNode_StateMachine_3
    bAlphaBoolEnabled=Get bShouldAdjustBodyTwistForMelee() -> (bShouldAdjustBodyTwistForMelee:bool)
   ) {alphaInputType=Bool, alphaBoolBlend {blendInTime=0.25, blendOption=Linear, blendOutTime=0.25}}
   n4 = AimOffset Player 'HardSaddleWolf_SlopePitch_AO'!(
    X:float=Get SlopePitch() -> (SlopePitch:float)
    BasePose:Pose Link Structure=Use cached pose 'PostUpperBodyPose'()
    bAlphaBoolEnabled:bool=NOT Boolean(A=Get bIsSlopeSliding() -> (bIsSlopeSliding:bool))
   ) -> (Pose:Pose Link Structure) {alphaInputType=Bool, blendSpace=&/HardSaddleWolf_SlopePitch_AO, alphaBoolBlend {blendOption=Linear}}
   n5 = AimOffset Player 'HardSaddleWolf_Player_HeadAO'!(X:float=Get InPlaceYaw() -> (InPlaceYaw:float), BasePose:Pose Link Structure=@n4, Alpha:float=Get HeadAOAlpha() -> (HeadAOAlpha:float)) -> (Pose:Pose Link Structure) {blendSpace=/RidingContent/Animation/HardSaddle/Wolf/AO/HardSaddleWolf_Player_HeadAO, alphaBoolBlend {blendOption=Linear}}
   n6 = Get FalseBlendTime() -> (FalseBlendTime:float)
   n7 = Get bIsRideMode() -> (bIsRideMode:bool)
   n8 = "Transform (Modify) Bone - Bone: pelvis"!(
    ComponentPose=Local To Component(LocalPose=Blend Poses by bool(BlendPose_0=@n5, BlendPose_1=@n3, BlendTime_0=0.25, BlendTime_1=@n6, bActiveValue=@n7) {blendProfile=&Fortnite_M_Avg_Player_Skeleton:FastWeapons, blendType=Linear})
    Alpha=0.25
    Rotation=Get MeleeTwistRot() -> (MeleeTwistRot:Rotator)
   ) -> (Pose) {
    rotationMode=BMM_Additive
    rotationSpace=BCS_WorldSpace
    translationSpace=BCS_WorldSpace
    alphaBoolBlend {blendOption=Linear}
    boneToModify {boneName=pelvis}
   }
   n9 = "Transform (Modify) Bone - Bone: weapon_r"!(ComponentPose=@n8, bAlphaBoolEnabled:bool=True, Scale="0.001,0.001,0.001") -> (Pose) {alphaInputType=Bool, scaleMode=BMM_Replace, alphaBoolBlend {blendOption=Linear}, boneToModify {boneName=weapon_r}, scale=<0.001 0.001 0.001>}
   n10 = "Transform (Modify) Bone - Bone: weapon_l"!(ComponentPose=@n9, bAlphaBoolEnabled:bool=True, Scale="0.001,0.001,0.001") -> (Pose) {alphaInputType=Bool, scaleMode=BMM_Replace, alphaBoolBlend {blendOption=Linear}, boneToModify {boneName=weapon_l}, scale=<0.001 0.001 0.001>}
   n11 = "Transform (Modify) Bone - Bone: hand_attach_r"!(ComponentPose=@n10, bAlphaBoolEnabled:bool=True, Scale="0.001,0.001,0.001") -> (Pose) {
    alphaInputType=Bool
    scaleMode=BMM_Replace
    alphaBoolBlend {blendOption=Linear}
    boneToModify {boneName=hand_attach_r}
    scale=<0.001 0.001 0.001>
   }
   n12 = "Transform (Modify) Bone - Bone: hand_attach_l"!(ComponentPose=@n11, bAlphaBoolEnabled:bool=True, Scale="0.001,0.001,0.001") -> (Pose) {
    alphaInputType=Bool
    scaleMode=BMM_Replace
    alphaBoolBlend {blendOption=Linear}
    boneToModify {boneName=hand_attach_l}
    scale=<0.001 0.001 0.001>
   }
   Output Pose(Result=Inertialization!(Source:Pose Link Structure=Component To Local(ComponentPose=@n12)) -> (Pose:Pose Link Structure) {is=title AnimGraphNode_Inertialization}) {layerGroup=VehicleOverride}
   Save cached pose 'UpperAndLowerPose'(Pose=Input Pose() {name=InPoseFullBody}) {cacheName=UpperAndLowerPose}
   AnimGraphNode_StateMachine_1 = UpperBody() {is=title AnimGraphNode_StateMachine, bAllowConduitEntryStates=true, bSkipFirstUpdateTransition=false, maxTransitionsPerFrame=1}
   n13 = Layered blend per bone(BasePose=Use cached pose 'FullBodyPose'(), BlendPoses_0=@AnimGraphNode_StateMachine_1) {
    bBlendRootMotionBasedOnRootBone=false
    bMeshSpaceRotationBlend=true
    layerSetup [{branchFilters [
       {boneName=Root}
       {blendDepth=-1, boneName=pelvis}
       {boneName=Spine_01}
       {blendDepth=-1, boneName=thigh_l}
       {blendDepth=-1, boneName=thigh_r}
       {blendDepth=-1, boneName=ik_foot_root}
      ]}]
   }
   Save cached pose 'PostUpperBodyPose'(Pose=@n13) {cacheName=PostUpperBodyPose}
   AnimGraphNode_StateMachine_0 = FullBody() {is=title AnimGraphNode_StateMachine, bAllowConduitEntryStates=true, bSkipFirstUpdateTransition=false}
   Save cached pose 'FullBodyPose'(Pose=@AnimGraphNode_StateMachine_0) {cacheName=FullBodyPose}
  ]}
 FullBody {
  entry=Idle
  path=VehicleFullBodyOverride.AnimGraphNode_StateMachine_0.FullBody
  ids=id
  nodes [
   SA_ToMoving() {is=title AnimStateAliasNode, also {aliases [Idle,JumpLand,Stop,Turn,TurnSettle]}}
   SA_ToJumpStart() {is=title AnimStateAliasNode, also {aliases [Idle,JumpLand,Moving,Stop,Turn,TurnSettle]}}
   SA_ToSwimming() {is=title AnimStateAliasNode, bGlobalAlias=true}
   SA_ToJumpFall() {is=title AnimStateAliasNode, also {aliases [JumpLand,Moving,Stop,Turn,TurnSettle]}}
   AnimStateNode_0 = JumpStart() {is=title AnimStateNode}
   AnimStateNode_1 = Stop() {is=title AnimStateNode}
   AnimStateNode_2 = Idle() {
    stateEntered {guid=#1}
    stateFullyBlended {guid=#2}
    stateLeft {guid=#3}
   }
   AnimStateNode_3 = Moving() {
    is=title AnimStateNode
    stateEntered {guid=#4}
    stateFullyBlended {guid=#5}
    stateLeft {guid=#6}
   }
   AnimStateNode_4 = JumpFall() {is=title AnimStateNode}
   AnimStateNode_5 = Turn() {is=title AnimStateNode}
   AnimStateNode_6 = JumpLand() {is=title AnimStateNode}
   AnimStateNode_7 = Swimming() {is=title AnimStateNode}
   AnimStateNode_8 = TurnSettle() {is=title AnimStateNode}
   AnimStateTransitionNode_0 = AnimStateTransitionNode() {crossfadeDuration=0.3, logicType=TLT_Inertialization, priorityOrder=100}
   AnimStateTransitionNode_1 = AnimStateTransitionNode() {bAutomaticRuleBasedOnSequencePlayerInState=true, crossfadeDuration=0.4, logicType=TLT_Inertialization}
   AnimStateTransitionNode_10 = AnimStateTransitionNode() {logicType=TLT_Inertialization, priorityOrder=2}
   AnimStateTransitionNode_11 = AnimStateTransitionNode() {blendMode=Linear, crossfadeDuration=0.3, priorityOrder=2}
   AnimStateTransitionNode_12 = AnimStateTransitionNode() {crossfadeDuration=0.4, logicType=TLT_Inertialization}
   AnimStateTransitionNode_13 = AnimStateTransitionNode() {crossfadeDuration=0.3, logicType=TLT_Inertialization}
   AnimStateTransitionNode_14 = AnimStateTransitionNode() {logicType=TLT_Inertialization}
   AnimStateTransitionNode_15 = AnimStateTransitionNode() {crossfadeDuration=0.6, logicType=TLT_Inertialization, priorityOrder=2}
   AnimStateTransitionNode_16 = AnimStateTransitionNode() {bAutomaticRuleBasedOnSequencePlayerInState=true, crossfadeDuration=0.3, priorityOrder=2}
   AnimStateTransitionNode_17 = AnimStateTransitionNode()
   AnimStateTransitionNode_2 = AnimStateTransitionNode() {crossfadeDuration=0.35, logicType=TLT_Inertialization, priorityOrder=3}
   AnimStateTransitionNode_3 = AnimStateTransitionNode() {bAutomaticRuleBasedOnSequencePlayerInState=true, crossfadeDuration=0.4, priorityOrder=3}
   AnimStateTransitionNode_4 = AnimStateTransitionNode() {logicType=TLT_Inertialization, priorityOrder=100}
   AnimStateTransitionNode_5 = AnimStateTransitionNode() {bAutomaticRuleBasedOnSequencePlayerInState=true, crossfadeDuration=0.3, logicType=TLT_Inertialization, priorityOrder=100}
   AnimStateTransitionNode_6 = AnimStateTransitionNode() {crossfadeDuration=0.4, logicType=TLT_Inertialization}
   AnimStateTransitionNode_7 = AnimStateTransitionNode() {logicType=TLT_Inertialization}
   AnimStateTransitionNode_8 = AnimStateTransitionNode()
   AnimStateTransitionNode_9 = AnimStateTransitionNode() {crossfadeDuration=0.15, logicType=TLT_Inertialization}
  ]
  links [
   {
    dst=Turn
    rule=VehicleFullBodyOverride.AnimGraphNode_StateMachine_0.FullBody.AnimStateTransitionNode_8.Transition
    src=Idle
    via=AnimStateTransitionNode_8
   }
   {
    dst=JumpLand
    rule=VehicleFullBodyOverride.AnimGraphNode_StateMachine_0.FullBody.AnimStateTransitionNode_4.Transition
    src=JumpFall
    via=AnimStateTransitionNode_4
   }
   {
    dst=Moving
    rule=VehicleFullBodyOverride.AnimGraphNode_StateMachine_0.FullBody.AnimStateTransitionNode_9.Transition
    src=JumpFall
    via=AnimStateTransitionNode_9
   }
   {
    dst=Idle
    rule=VehicleFullBodyOverride.AnimGraphNode_StateMachine_0.FullBody.AnimStateTransitionNode_5.Transition
    src=JumpLand
    via=AnimStateTransitionNode_5
   }
   {
    dst=Moving
    rule=VehicleFullBodyOverride.AnimGraphNode_StateMachine_0.FullBody.AnimStateTransitionNode_10.Transition
    src=JumpLand
    via=AnimStateTransitionNode_10
   }
   {
    dst=JumpFall
    rule=VehicleFullBodyOverride.AnimGraphNode_StateMachine_0.FullBody.AnimStateTransitionNode_1.Transition
    src=JumpStart
    via=AnimStateTransitionNode_1
   }
   {
    dst=Stop
    rule=VehicleFullBodyOverride.AnimGraphNode_StateMachine_0.FullBody.AnimStateTransitionNode_2.Transition
    src=Moving
    via=AnimStateTransitionNode_2
   }
   {
    dst=JumpFall
    rule=VehicleFullBodyOverride.AnimGraphNode_StateMachine_0.FullBody.AnimStateTransitionNode_13.Transition
    src=SA_ToJumpFall
    via=AnimStateTransitionNode_13
   }
   {
    dst=JumpStart
    rule=VehicleFullBodyOverride.AnimGraphNode_StateMachine_0.FullBody.AnimStateTransitionNode_0.Transition
    src=SA_ToJumpStart
    via=AnimStateTransitionNode_0
   }
   {
    dst=Moving
    rule=VehicleFullBodyOverride.AnimGraphNode_StateMachine_0.FullBody.AnimStateTransitionNode_15.Transition
    src=SA_ToMoving
    via=AnimStateTransitionNode_15
   }
   {
    dst=Swimming
    rule=VehicleFullBodyOverride.AnimGraphNode_StateMachine_0.FullBody.AnimStateTransitionNode_6.Transition
    src=SA_ToSwimming
    via=AnimStateTransitionNode_6
   }
   {
    dst=Idle
    rule=VehicleFullBodyOverride.AnimGraphNode_StateMachine_0.FullBody.AnimStateTransitionNode_3.Transition
    src=Stop
    via=AnimStateTransitionNode_3
   }
   {
    dst=Moving
    rule=VehicleFullBodyOverride.AnimGraphNode_StateMachine_0.FullBody.AnimStateTransitionNode_11.Transition
    src=Stop
    via=AnimStateTransitionNode_11
   }
   {
    dst=Turn
    rule=VehicleFullBodyOverride.AnimGraphNode_StateMachine_0.FullBody.AnimStateTransitionNode_12.Transition
    src=Stop
    via=AnimStateTransitionNode_12
   }
   {
    dst=Moving
    rule=VehicleFullBodyOverride.AnimGraphNode_StateMachine_0.FullBody.AnimStateTransitionNode_7.Transition
    src=Swimming
    via=AnimStateTransitionNode_7
   }
   {
    dst=TurnSettle
    rule=VehicleFullBodyOverride.AnimGraphNode_StateMachine_0.FullBody.AnimStateTransitionNode_14.Transition
    src=Turn
    via=AnimStateTransitionNode_14
   }
   {
    dst=Idle
    rule=VehicleFullBodyOverride.AnimGraphNode_StateMachine_0.FullBody.AnimStateTransitionNode_16.Transition
    src=TurnSettle
    via=AnimStateTransitionNode_16
   }
   {
    dst=Turn
    rule=VehicleFullBodyOverride.AnimGraphNode_StateMachine_0.FullBody.AnimStateTransitionNode_17.Transition
    src=TurnSettle
    via=AnimStateTransitionNode_17
   }
  ]
 }
 JumpStart {
  path=VehicleFullBodyOverride.AnimGraphNode_StateMachine_0.FullBody.AnimStateNode_0.JumpStart
  nodes [Output Animation Pose(Result=Sequence Player 'HardSaddleWolf_Player_Jump_Start'() {bLoopAnimation=false, sequence=&/HardSaddleWolf_Player_Jump_Start})]
 }
 Stop {
  path=VehicleFullBodyOverride.AnimGraphNode_StateMachine_0.FullBody.AnimStateNode_1.Stop
  nodes [
   AnimGraphNode_StateMachine_0 = StopTransition() {is=title AnimGraphNode_StateMachine, bAllowConduitEntryStates=true, bSkipFirstUpdateTransition=false, maxTransitionsPerFrame=1}
   Output Animation Pose(Result=@AnimGraphNode_StateMachine_0)
  ]
 }
 StopTransition {
  entry=Conduit
  path=VehicleFullBodyOverride.AnimGraphNode_StateMachine_0.FullBody.AnimStateNode_1.Stop.AnimGraphNode_StateMachine_0.StopTransition
  ids=id
  nodes [
   AnimStateConduitNode_0 = Conduit()
   AnimStateNode_0 = FeetInAir() {is=title AnimStateNode}
   AnimStateNode_1 = FrontFeetPlanted() {is=title AnimStateNode}
   AnimStateNode_2 = BackFeetPlanted() {is=title AnimStateNode}
   AnimStateNode_3 = Backwards() {is=title AnimStateNode}
   AnimStateTransitionNode_0 = AnimStateTransitionNode() {logicType=TLT_Inertialization, priorityOrder=0}
   AnimStateTransitionNode_3 = AnimStateTransitionNode() {logicType=TLT_Inertialization}
   AnimStateTransitionNode_4 = AnimStateTransitionNode() {logicType=TLT_Inertialization, priorityOrder=2}
   AnimStateTransitionNode_5 = AnimStateTransitionNode() {logicType=TLT_Inertialization, priorityOrder=100}
  ]
  links [
   {
    dst=BackFeetPlanted
    rule=VehicleFullBodyOverride.AnimGraphNode_StateMachine_0.FullBody.AnimStateNode_1.Stop.AnimGraphNode_StateMachine_0.StopTransition.AnimStateTransitionNode_5.Transition
    src=Conduit
    via=AnimStateTransitionNode_5
   }
   {
    dst=Backwards
    rule=VehicleFullBodyOverride.AnimGraphNode_StateMachine_0.FullBody.AnimStateNode_1.Stop.AnimGraphNode_StateMachine_0.StopTransition.AnimStateTransitionNode_0.Transition
    src=Conduit
    via=AnimStateTransitionNode_0
   }
   {
    dst=FeetInAir
    rule=VehicleFullBodyOverride.AnimGraphNode_StateMachine_0.FullBody.AnimStateNode_1.Stop.AnimGraphNode_StateMachine_0.StopTransition.AnimStateTransitionNode_3.Transition
    src=Conduit
    via=AnimStateTransitionNode_3
   }
   {
    dst=FrontFeetPlanted
    rule=VehicleFullBodyOverride.AnimGraphNode_StateMachine_0.FullBody.AnimStateNode_1.Stop.AnimGraphNode_StateMachine_0.StopTransition.AnimStateTransitionNode_4.Transition
    src=Conduit
    via=AnimStateTransitionNode_4
   }
  ]
 }
 Conduit {
  path=VehicleFullBodyOverride.AnimGraphNode_StateMachine_0.FullBody.AnimStateNode_1.Stop.AnimGraphNode_StateMachine_0.StopTransition.AnimStateConduitNode_0.Conduit
  nodes [Result(bCanEnterTransition=true)]
 }
 FeetInAir {
  path=VehicleFullBodyOverride.AnimGraphNode_StateMachine_0.FullBody.AnimStateNode_1.Stop.AnimGraphNode_StateMachine_0.StopTransition.AnimStateNode_0.FeetInAir
  nodes [Output Animation Pose(Result=Sequence Player 'HardSaddleWolf_Player_Stop_FeetInAir'() {bLoopAnimation=false, sequence=&/HardSaddleWolf_Player_Stop_FeetInAir})]
 }
 FrontFeetPlanted {
  path=VehicleFullBodyOverride.AnimGraphNode_StateMachine_0.FullBody.AnimStateNode_1.Stop.AnimGraphNode_StateMachine_0.StopTransition.AnimStateNode_1.FrontFeetPlanted
  nodes [Output Animation Pose(Result=Sequence Player 'HardSaddleWolf_Player_Stop_FrontFeetPlanted'() {bLoopAnimation=false, sequence=&/HardSaddleWolf_Player_Stop_FrontFeetPlanted})]
 }
 BackFeetPlanted {
  path=VehicleFullBodyOverride.AnimGraphNode_StateMachine_0.FullBody.AnimStateNode_1.Stop.AnimGraphNode_StateMachine_0.StopTransition.AnimStateNode_2.BackFeetPlanted
  nodes [Output Animation Pose(Result=Sequence Player 'HardSaddleWolf_Player_Stop_BackFeetPlanted'() {bLoopAnimation=false, sequence=&/HardSaddleWolf_Player_Stop_BackFeetPlanted})]
 }
 Backwards {
  path=VehicleFullBodyOverride.AnimGraphNode_StateMachine_0.FullBody.AnimStateNode_1.Stop.AnimGraphNode_StateMachine_0.StopTransition.AnimStateNode_3.Backwards
  nodes [
   AnimGraphNode_StateMachine_0 = StopBackwards() {is=title AnimGraphNode_StateMachine, bAllowConduitEntryStates=true, bSkipFirstUpdateTransition=false, maxTransitionsPerFrame=1}
   Output Animation Pose(Result=@AnimGraphNode_StateMachine_0)
  ]
 }
 StopBackwards {
  entry=Conduit
  path=VehicleFullBodyOverride.AnimGraphNode_StateMachine_0.FullBody.AnimStateNode_1.Stop.AnimGraphNode_StateMachine_0.StopTransition.AnimStateNode_3.Backwards.AnimGraphNode_StateMachine_0.StopBackwards
  ids=id
  nodes [
   AnimStateConduitNode_0 = Conduit()
   AnimStateNode_0 = LeftBackFeetForward() {is=title AnimStateNode}
   AnimStateNode_1 = RightBackFeetForward() {is=title AnimStateNode}
   AnimStateNode_2 = LeftPlantedRightPass() {is=title AnimStateNode}
   AnimStateNode_3 = RightPlantedLeftPass() {is=title AnimStateNode}
   AnimStateTransitionNode_0 = AnimStateTransitionNode() {logicType=TLT_Inertialization}
   AnimStateTransitionNode_1 = AnimStateTransitionNode() {logicType=TLT_Inertialization, priorityOrder=2}
   AnimStateTransitionNode_2 = AnimStateTransitionNode() {logicType=TLT_Inertialization, priorityOrder=3}
   AnimStateTransitionNode_3 = AnimStateTransitionNode() {logicType=TLT_Inertialization, priorityOrder=100}
  ]
  links [
   {
    dst=LeftBackFeetForward
    rule=VehicleFullBodyOverride.AnimGraphNode_StateMachine_0.FullBody.AnimStateNode_1.Stop.AnimGraphNode_StateMachine_0.StopTransition.AnimStateNode_3.Backwards.AnimGraphNode_StateMachine_0.StopBackwards.AnimStateTransitionNode_0.AnimationTransitionGraph_1
    src=Conduit
    via=AnimStateTransitionNode_0
   }
   {
    dst=LeftPlantedRightPass
    rule=VehicleFullBodyOverride.AnimGraphNode_StateMachine_0.FullBody.AnimStateNode_1.Stop.AnimGraphNode_StateMachine_0.StopTransition.AnimStateNode_3.Backwards.AnimGraphNode_StateMachine_0.StopBackwards.AnimStateTransitionNode_2.AnimationTransitionGraph_1
    src=Conduit
    via=AnimStateTransitionNode_2
   }
   {
    dst=RightBackFeetForward
    rule=VehicleFullBodyOverride.AnimGraphNode_StateMachine_0.FullBody.AnimStateNode_1.Stop.AnimGraphNode_StateMachine_0.StopTransition.AnimStateNode_3.Backwards.AnimGraphNode_StateMachine_0.StopBackwards.AnimStateTransitionNode_1.AnimationTransitionGraph_1
    src=Conduit
    via=AnimStateTransitionNode_1
   }
   {
    dst=RightPlantedLeftPass
    rule=VehicleFullBodyOverride.AnimGraphNode_StateMachine_0.FullBody.AnimStateNode_1.Stop.AnimGraphNode_StateMachine_0.StopTransition.AnimStateNode_3.Backwards.AnimGraphNode_StateMachine_0.StopBackwards.AnimStateTransitionNode_3.AnimationTransitionGraph_1
    src=Conduit
    via=AnimStateTransitionNode_3
   }
  ]
 }
 Conduit {
  path=VehicleFullBodyOverride.AnimGraphNode_StateMachine_0.FullBody.AnimStateNode_1.Stop.AnimGraphNode_StateMachine_0.StopTransition.AnimStateNode_3.Backwards.AnimGraphNode_StateMachine_0.StopBackwards.AnimStateConduitNode_0.Conduit
  nodes [Result(bCanEnterTransition=true)]
 }
 LeftBackFeetForward {
  path=VehicleFullBodyOverride.AnimGraphNode_StateMachine_0.FullBody.AnimStateNode_1.Stop.AnimGraphNode_StateMachine_0.StopTransition.AnimStateNode_3.Backwards.AnimGraphNode_StateMachine_0.StopBackwards.AnimStateNode_0.LeftBackFeetForward
  nodes [Output Animation Pose(Result=Sequence Player 'HardSaddleWolf_Player_Stop_Backwards'() {bLoopAnimation=false, sequence=&/HardSaddleWolf_Player_Stop_Backwards})]
 }
 RightBackFeetForward {
  path=VehicleFullBodyOverride.AnimGraphNode_StateMachine_0.FullBody.AnimStateNode_1.Stop.AnimGraphNode_StateMachine_0.StopTransition.AnimStateNode_3.Backwards.AnimGraphNode_StateMachine_0.StopBackwards.AnimStateNode_1.RightBackFeetForward
  nodes [Output Animation Pose(Result=Sequence Player 'HardSaddleWolf_Player_Stop_Backwards_fr19'() {bLoopAnimation=false, sequence=&/HardSaddleWolf_Player_Stop_Backwards_fr19})]
 }
 LeftPlantedRightPass {
  path=VehicleFullBodyOverride.AnimGraphNode_StateMachine_0.FullBody.AnimStateNode_1.Stop.AnimGraphNode_StateMachine_0.StopTransition.AnimStateNode_3.Backwards.AnimGraphNode_StateMachine_0.StopBackwards.AnimStateNode_2.LeftPlantedRightPass
  nodes [Output Animation Pose(Result=Sequence Player 'HardSaddleWolf_Player_Stop_Backwards_fr24'() {bLoopAnimation=false, sequence=&/HardSaddleWolf_Player_Stop_Backwards_fr24})]
 }
 RightPlantedLeftPass {
  path=VehicleFullBodyOverride.AnimGraphNode_StateMachine_0.FullBody.AnimStateNode_1.Stop.AnimGraphNode_StateMachine_0.StopTransition.AnimStateNode_3.Backwards.AnimGraphNode_StateMachine_0.StopBackwards.AnimStateNode_3.RightPlantedLeftPass
  nodes [Output Animation Pose(Result=Sequence Player 'HardSaddleWolf_Player_Stop_Backwards_fr7'() {bLoopAnimation=false, sequence=&/HardSaddleWolf_Player_Stop_Backwards_fr7})]
 }
 AnimationTransitionGraph_1 {
  path=VehicleFullBodyOverride.AnimGraphNode_StateMachine_0.FullBody.AnimStateNode_1.Stop.AnimGraphNode_StateMachine_0.StopTransition.AnimStateNode_3.Backwards.AnimGraphNode_StateMachine_0.StopBackwards.AnimStateTransitionNode_0.AnimationTransitionGraph_1
  nodes [Result()]
 }
 AnimationTransitionGraph_1 {
  path=VehicleFullBodyOverride.AnimGraphNode_StateMachine_0.FullBody.AnimStateNode_1.Stop.AnimGraphNode_StateMachine_0.StopTransition.AnimStateNode_3.Backwards.AnimGraphNode_StateMachine_0.StopBackwards.AnimStateTransitionNode_1.AnimationTransitionGraph_1
  nodes [Result()]
 }
 AnimationTransitionGraph_1 {
  path=VehicleFullBodyOverride.AnimGraphNode_StateMachine_0.FullBody.AnimStateNode_1.Stop.AnimGraphNode_StateMachine_0.StopTransition.AnimStateNode_3.Backwards.AnimGraphNode_StateMachine_0.StopBackwards.AnimStateTransitionNode_2.AnimationTransitionGraph_1
  nodes [Result()]
 }
 AnimationTransitionGraph_1 {
  path=VehicleFullBodyOverride.AnimGraphNode_StateMachine_0.FullBody.AnimStateNode_1.Stop.AnimGraphNode_StateMachine_0.StopTransition.AnimStateNode_3.Backwards.AnimGraphNode_StateMachine_0.StopBackwards.AnimStateTransitionNode_3.AnimationTransitionGraph_1
  nodes [Result(bCanEnterTransition=true)]
 }
 Transition {
  path=VehicleFullBodyOverride.AnimGraphNode_StateMachine_0.FullBody.AnimStateNode_1.Stop.AnimGraphNode_StateMachine_0.StopTransition.AnimStateTransitionNode_0.Transition
  nodes [Result(bCanEnterTransition=Get Should Play Stop Backwards() -> (Should Play Stop Backwards:bool))]
 }
 Transition {
  path=VehicleFullBodyOverride.AnimGraphNode_StateMachine_0.FullBody.AnimStateNode_1.Stop.AnimGraphNode_StateMachine_0.StopTransition.AnimStateTransitionNode_3.Transition
  nodes [Result()]
 }
 Transition {
  path=VehicleFullBodyOverride.AnimGraphNode_StateMachine_0.FullBody.AnimStateNode_1.Stop.AnimGraphNode_StateMachine_0.StopTransition.AnimStateTransitionNode_4.Transition
  nodes [Result()]
 }
 Transition {
  path=VehicleFullBodyOverride.AnimGraphNode_StateMachine_0.FullBody.AnimStateNode_1.Stop.AnimGraphNode_StateMachine_0.StopTransition.AnimStateTransitionNode_5.Transition
  nodes [Result(bCanEnterTransition=true)]
 }
 Idle {
  path=VehicleFullBodyOverride.AnimGraphNode_StateMachine_0.FullBody.AnimStateNode_2.Idle
  nodes [Output Animation Pose(Result=Sequence Player 'HardSaddleWolf_Player_Idle'() {sequence=&/HardSaddleWolf_Player_Idle})]
 }
 Moving {
  path=VehicleFullBodyOverride.AnimGraphNode_StateMachine_0.FullBody.AnimStateNode_3.Moving
  nodes [
   n1 = Get RidingMovePlayRate!() -> (RidingMovePlayRate)
   n2 = Get WolfSpeed2D()
   n3 = Get BoostPlayRate() -> (BoostPlayRate:float)
   n4 = Apply Additive!(
    Base=Blend Poses by bool(
     BlendPose_0=Sequence Player 'HardSaddle_Player_Boost_N'!(PlayRate:float=@n3) -> (Pose) {groupName=Locomotion, method=SyncGroup, sequence=&/HardSaddle_Player_Boost_N}
     BlendPose_1=Blendspace Player 'HardSaddleWolf_Player_WalkRun_BS'!(X=@n2, PlayRate:float=@n1) -> (Pose) {blendSpace=&/HardSaddleWolf_Player_WalkRun_BS, groupName=Locomotion, method=SyncGroup}
     BlendTime_0=0.25
     BlendTime_1=0.25
     bActiveValue=Get Is Boosting() -> (Is Boosting:bool)
    ) {blendType=Linear}
    Additive=Blendspace Player 'HardSaddleWolf_Player_BodyRoll_BS'(X=Get BodyRoll!() -> (BodyRoll:double), Y=Get WolfSpeed2D()) {blendSpace=&/HardSaddleWolf_Player_BodyRoll_BS}
    Alpha:float
   ) -> (Pose) {alphaBoolBlend {blendInTime=0.5, blendOption=Linear, blendOutTime=0.35}}
   n5 = Sequence Player 'HardSaddleWolf_Player_Walk_S'!(PlayRate:float=Get Backwards Play Rate() -> (Backwards Play Rate:double)) -> (Pose) {groupName=Locomotion, method=SyncGroup, sequence=&/HardSaddleWolf_Player_Walk_S}
   n6 = Get Is Moving Backwards() -> (Is Moving Backwards:bool)
   AnimGraphNode_StateMachine_0 = MovingJumpLand() {is=title AnimGraphNode_StateMachine}
   Output Animation Pose(Result=Apply Additive(
     Base=Blend Poses by bool(BlendPose_0=@n5, BlendPose_1=@n4, BlendTime_0=0.3, BlendTime_1=0.3, bActiveValue=@n6) {blendType=Linear}
     Additive=@AnimGraphNode_StateMachine_0
    ) {alphaBoolBlend {blendInTime=0.5, blendOption=Linear, blendOutTime=0.35}})
  ]
 }
 MovingJumpLand {
  entry=RefAdd
  path=VehicleFullBodyOverride.AnimGraphNode_StateMachine_0.FullBody.AnimStateNode_3.Moving.AnimGraphNode_StateMachine_0.MovingJumpLand
  ids=id
  nodes [
   AnimStateNode_0 = RefAdd() {is=title AnimStateNode}
   AnimStateNode_1 = MovingJumpAdd() {is=title AnimStateNode}
   AnimStateTransitionNode_0 = AnimStateTransitionNode() {crossfadeDuration=0.067, logicType=TLT_Inertialization}
   AnimStateTransitionNode_1 = AnimStateTransitionNode() {bAutomaticRuleBasedOnSequencePlayerInState=true, crossfadeDuration=0.3, logicType=TLT_Inertialization}
  ]
  links [
   {
    dst=RefAdd
    rule=VehicleFullBodyOverride.AnimGraphNode_StateMachine_0.FullBody.AnimStateNode_3.Moving.AnimGraphNode_StateMachine_0.MovingJumpLand.AnimStateTransitionNode_1.Transition
    src=MovingJumpAdd
    via=AnimStateTransitionNode_1
   }
   {
    dst=MovingJumpAdd
    rule=VehicleFullBodyOverride.AnimGraphNode_StateMachine_0.FullBody.AnimStateNode_3.Moving.AnimGraphNode_StateMachine_0.MovingJumpLand.AnimStateTransitionNode_0.Transition
    src=RefAdd
    via=AnimStateTransitionNode_0
   }
  ]
 }
 RefAdd {
  path=VehicleFullBodyOverride.AnimGraphNode_StateMachine_0.FullBody.AnimStateNode_3.Moving.AnimGraphNode_StateMachine_0.MovingJumpLand.AnimStateNode_0.RefAdd
  nodes [Output Animation Pose(Result=Additive Identity Pose())]
 }
 MovingJumpAdd {
  path=VehicleFullBodyOverride.AnimGraphNode_StateMachine_0.FullBody.AnimStateNode_3.Moving.AnimGraphNode_StateMachine_0.MovingJumpLand.AnimStateNode_1.MovingJumpAdd
  nodes [Output Animation Pose(Result="Sequence Player (additive) 'HardSaddleWolf_Player_MovingJumpLand_Add'"() {bLoopAnimation=false, sequence=&/HardSaddleWolf_Player_MovingJumpLand_Add})]
 }
 Transition {
  path=VehicleFullBodyOverride.AnimGraphNode_StateMachine_0.FullBody.AnimStateNode_3.Moving.AnimGraphNode_StateMachine_0.MovingJumpLand.AnimStateTransitionNode_0.Transition
  nodes [Result(bCanEnterTransition=Get Play Land Additive() -> (Play Land Additive:bool))]
 }
 Transition {
  path=VehicleFullBodyOverride.AnimGraphNode_StateMachine_0.FullBody.AnimStateNode_3.Moving.AnimGraphNode_StateMachine_0.MovingJumpLand.AnimStateTransitionNode_1.Transition
  nodes [Result()]
 }
 JumpFall {
  path=VehicleFullBodyOverride.AnimGraphNode_StateMachine_0.FullBody.AnimStateNode_4.JumpFall
  nodes [
   n1 = Blendspace Player 'HardSaddleWolf_Player_SlopeSliding_BS'(X=Get SlopeSlidingRoll() -> (SlopeSlidingRoll:float), Y=Get SlopeSlidingPitch() -> (SlopeSlidingPitch:float)) {blendSpace=/RidingContent/Animation/HardSaddle/Wolf/SlopeSliding/HardSaddleWolf_Player_SlopeSliding_BS}
   n2 = "Sequence Player (additive) 'HardSaddleWolf_Player_SlopeSliding_Add'"() {sequence=/RidingContent/Animation/HardSaddle/Wolf/SlopeSliding/HardSaddleWolf_Player_SlopeSliding_Add}
   Output Animation Pose(Result=Blend Poses by bool(
     BlendPose_0=Apply Additive(Base=@n1, Additive=@n2) {alphaBoolBlend {blendOption=Linear}}
     BlendPose_1=Sequence Player 'HardSaddleWolf_Player_Jump_Loop'() {sequence=&/HardSaddleWolf_Player_Jump_Loop}
     bActiveValue=Get bIsSlopeSliding()
    ) {blendType=Linear})
  ]
 }
 Turn {
  path=VehicleFullBodyOverride.AnimGraphNode_StateMachine_0.FullBody.AnimStateNode_5.Turn
  nodes [Output Animation Pose(Result=Blend Poses by bool(
     BlendPose_0=Sequence Player 'HardSaddleWolf_Player_TurnRight'!(PlayRate:float=1) -> (Pose) {sequence=&/HardSaddleWolf_Player_TurnRight}
     BlendPose_1=Sequence Player 'HardSaddleWolf_Player_TurnLeft'!(PlayRate:float=1) -> (Pose) {sequence=&/HardSaddleWolf_Player_TurnLeft}
    ) {blendType=Linear})]
 }
 JumpLand {
  path=VehicleFullBodyOverride.AnimGraphNode_StateMachine_0.FullBody.AnimStateNode_6.JumpLand
  nodes [Output Animation Pose(Result=Sequence Player 'HardSaddleWolf_Player_Jump_Land'() {bLoopAnimation=false, sequence=&/HardSaddleWolf_Player_Jump_Land})]
 }
 Swimming {
  path=VehicleFullBodyOverride.AnimGraphNode_StateMachine_0.FullBody.AnimStateNode_7.Swimming
  nodes [
   n1 = Sequence Player 'HardSaddleWolf_Player_Swim_N'() {groupName=Locomotion, method=SyncGroup, sequence=&/HardSaddleWolf_Player_Swim_N}
   n2 = Sequence Player 'HardSaddleWolf_Player_Swim_S'() {groupName=Locomotion, method=SyncGroup, sequence=&/HardSaddleWolf_Player_Swim_S}
   n3 = Get Is Moving Backwards()
   Output Animation Pose(Result=Blend Poses by bool(
     BlendPose_0=Blend Poses by bool(BlendPose_0=@n2, BlendPose_1=@n1, BlendTime_0=0.35, BlendTime_1=0.35, bActiveValue=@n3) {blendType=Linear, transitionType=Inertialization}
     BlendPose_1=Sequence Player 'HardSaddleWolf_Player_Idle_Swim'() {sequence=&/HardSaddleWolf_Player_Idle_Swim}
     BlendTime_0=0.5
     BlendTime_1=1
     bActiveValue=Get State Rule Swim Idle Swim Loco() -> (State Rule Swim Idle Swim Loco:bool)
    ) {blendType=Linear})
  ]
 }
 TurnSettle {
  path=VehicleFullBodyOverride.AnimGraphNode_StateMachine_0.FullBody.AnimStateNode_8.TurnSettle
  nodes [Output Animation Pose(Result=Blend Poses by bool(
     BlendPose_0=Sequence Player 'HardSaddleWolf_Player_TurnRight_Settle'() {bLoopAnimation=false, sequence=&/HardSaddleWolf_Player_TurnRight_Settle}
     BlendPose_1=Sequence Player 'HardSaddleWolf_Player_TurnLeft_Settle'() {bLoopAnimation=false, sequence=&/HardSaddleWolf_Player_TurnLeft_Settle}
     BlendTime_0=0.25
     BlendTime_1=0.25
     bActiveValue=Get Was Turning Right() -> (Was Turning Right:bool)
    ) {blendType=Linear})]
 }
 Transition {
  path=VehicleFullBodyOverride.AnimGraphNode_StateMachine_0.FullBody.AnimStateTransitionNode_0.Transition
  nodes [Result(bCanEnterTransition=Get Is Jumping() -> (Is Jumping:bool))]
 }
 Transition {path=VehicleFullBodyOverride.AnimGraphNode_StateMachine_0.FullBody.AnimStateTransitionNode_1.Transition, nodes [Result()]}
 Transition {
  path=VehicleFullBodyOverride.AnimGraphNode_StateMachine_0.FullBody.AnimStateTransitionNode_10.Transition
  nodes [Result(bCanEnterTransition=Get State Rule Jump Land Moving() -> (State Rule Jump Land Moving:bool))]
 }
 Transition {
  path=VehicleFullBodyOverride.AnimGraphNode_StateMachine_0.FullBody.AnimStateTransitionNode_11.Transition
  nodes [Result(bCanEnterTransition=Get State Rule Idle Moving() -> (State Rule Idle Moving:bool))]
 }
 Transition {
  path=VehicleFullBodyOverride.AnimGraphNode_StateMachine_0.FullBody.AnimStateTransitionNode_12.Transition
  nodes [Result(bCanEnterTransition=Get Is Turning in Place() -> (Is Turning in Place:bool))]
 }
 Transition {
  path=VehicleFullBodyOverride.AnimGraphNode_StateMachine_0.FullBody.AnimStateTransitionNode_13.Transition
  nodes [Result(bCanEnterTransition=Get Is Falling() -> (Is Falling:bool))]
 }
 Transition {
  path=VehicleFullBodyOverride.AnimGraphNode_StateMachine_0.FullBody.AnimStateTransitionNode_14.Transition
  nodes [Result(bCanEnterTransition=NOT Boolean(A=Get Is Turning in Place()))]
 }
 Transition {
  path=VehicleFullBodyOverride.AnimGraphNode_StateMachine_0.FullBody.AnimStateTransitionNode_15.Transition
  nodes [Result(bCanEnterTransition=Get State Rule Idle Moving())]
 }
 Transition {path=VehicleFullBodyOverride.AnimGraphNode_StateMachine_0.FullBody.AnimStateTransitionNode_16.Transition, nodes [Result()]}
 Transition {
  path=VehicleFullBodyOverride.AnimGraphNode_StateMachine_0.FullBody.AnimStateTransitionNode_17.Transition
  nodes [Result(bCanEnterTransition=Get Is Turning in Place())]
 }
 Transition {
  path=VehicleFullBodyOverride.AnimGraphNode_StateMachine_0.FullBody.AnimStateTransitionNode_2.Transition
  nodes [Result(bCanEnterTransition=Get State Rule Moving Idle() -> (State Rule Moving Idle:bool))]
 }
 Transition {path=VehicleFullBodyOverride.AnimGraphNode_StateMachine_0.FullBody.AnimStateTransitionNode_3.Transition, nodes [Result()]}
 Transition {
  path=VehicleFullBodyOverride.AnimGraphNode_StateMachine_0.FullBody.AnimStateTransitionNode_4.Transition
  nodes [Result(bCanEnterTransition=Get State Rule Fall Idle() -> (State Rule Fall Idle:bool))]
 }
 Transition {path=VehicleFullBodyOverride.AnimGraphNode_StateMachine_0.FullBody.AnimStateTransitionNode_5.Transition, nodes [Result()]}
 Transition {
  path=VehicleFullBodyOverride.AnimGraphNode_StateMachine_0.FullBody.AnimStateTransitionNode_6.Transition
  nodes [Result(bCanEnterTransition=Get Is Surface Swimming() -> (Is Surface Swimming:bool))]
 }
 Transition {
  path=VehicleFullBodyOverride.AnimGraphNode_StateMachine_0.FullBody.AnimStateTransitionNode_7.Transition
  nodes [Result(bCanEnterTransition=NOT Boolean(A=Get Is Surface Swimming()))]
 }
 Transition {
  path=VehicleFullBodyOverride.AnimGraphNode_StateMachine_0.FullBody.AnimStateTransitionNode_8.Transition
  nodes [Result(bCanEnterTransition=Get Is Turning in Place())]
 }
 Transition {
  path=VehicleFullBodyOverride.AnimGraphNode_StateMachine_0.FullBody.AnimStateTransitionNode_9.Transition
  nodes [Result(bCanEnterTransition=Get State Rule Fall Moving() -> (State Rule Fall Moving:bool))]
 }
 UpperBody {
  entry=FullBodyPose
  path=VehicleFullBodyOverride.AnimGraphNode_StateMachine_1.UpperBody
  ids=id
  nodes [
   AnimStateNode_0 = HackFullBodyPose() {is=title AnimStateNode}
   AnimStateNode_1 = CombatToRide() {is=title AnimStateNode}
   AnimStateNode_2 = FullBodyPose() {is=title AnimStateNode}
   AnimStateTransitionNode_5 = AnimStateTransitionNode() {crossfadeDuration=0.1, logicType=TLT_Inertialization}
   AnimStateTransitionNode_6 = AnimStateTransitionNode() {blendMode=Linear, crossfadeDuration=0, logicType=TLT_Inertialization}
   AnimStateTransitionNode_8 = AnimStateTransitionNode() {bAutomaticRuleBasedOnSequencePlayerInState=true, logicType=TLT_Inertialization}
  ]
  links [
   {
    dst=HackFullBodyPose
    rule=VehicleFullBodyOverride.AnimGraphNode_StateMachine_1.UpperBody.AnimStateTransitionNode_8.Transition
    src=CombatToRide
    via=AnimStateTransitionNode_8
   }
   {
    dst=CombatToRide
    rule=VehicleFullBodyOverride.AnimGraphNode_StateMachine_1.UpperBody.AnimStateTransitionNode_5.Transition
    src=FullBodyPose
    via=AnimStateTransitionNode_5
   }
   {
    dst=FullBodyPose
    rule=VehicleFullBodyOverride.AnimGraphNode_StateMachine_1.UpperBody.AnimStateTransitionNode_6.Transition
    src=HackFullBodyPose
    via=AnimStateTransitionNode_6
   }
  ]
 }
 HackFullBodyPose {
  path=VehicleFullBodyOverride.AnimGraphNode_StateMachine_1.UpperBody.AnimStateNode_0.HackFullBodyPose
  nodes [Output Animation Pose(Result=Use cached pose 'FullBodyPose'())]
 }
 CombatToRide {
  path=VehicleFullBodyOverride.AnimGraphNode_StateMachine_1.UpperBody.AnimStateNode_1.CombatToRide
  nodes [Output Animation Pose(Result=Blendspace Player 'HardSaddleWolf_Player_CtR_BS'!(X=Get CombatToRideYaw() -> (CombatToRideYaw:float)) -> (Pose) {bLoop=false, blendSpace=&/HardSaddleWolf_Player_CtR_BS})]
 }
 FullBodyPose {
  path=VehicleFullBodyOverride.AnimGraphNode_StateMachine_1.UpperBody.AnimStateNode_2.FullBodyPose
  nodes [Output Animation Pose(Result=Use cached pose 'FullBodyPose'())]
 }
 Transition {
  path=VehicleFullBodyOverride.AnimGraphNode_StateMachine_1.UpperBody.AnimStateTransitionNode_5.Transition
  nodes [Result(bCanEnterTransition=Get bIsRideMode())]
 }
 Transition {
  path=VehicleFullBodyOverride.AnimGraphNode_StateMachine_1.UpperBody.AnimStateTransitionNode_6.Transition
  nodes [Result(bCanEnterTransition=NOT Boolean(A=Get bIsRideMode()))]
 }
 Transition {path=VehicleFullBodyOverride.AnimGraphNode_StateMachine_1.UpperBody.AnimStateTransitionNode_8.Transition, nodes [Result()]}
 CombatUpperBodyAdditive {
  entry=Idle
  path=VehicleFullBodyOverride.AnimGraphNode_StateMachine_2.CombatUpperBodyAdditive
  ids=id
  nodes [
   SA_ToSwimming() {bGlobalAlias=true}
   SA_ToMoving() {also {aliases [Idle,JumpLand,Stop,Turn,TurnSettle]}}
   SA_ToJumpStart() {also {aliases [Idle,JumpLand,Moving,Stop,Turn,TurnSettle]}}
   SA_ToJumpFall() {also {aliases [JumpLand,Moving,Stop,Turn,TurnSettle]}}
   AnimStateNode_0 = Moving()
   AnimStateNode_1 = Turn()
   AnimStateNode_2 = Swimming()
   AnimStateNode_3 = TurnSettle()
   AnimStateNode_4 = Idle()
   AnimStateNode_5 = JumpStart()
   AnimStateNode_6 = JumpFall()
   AnimStateNode_7 = Stop()
   AnimStateNode_8 = JumpLand()
   AnimStateTransitionNode_0 = AnimStateTransitionNode()
   AnimStateTransitionNode_1 = AnimStateTransitionNode() {bAutomaticRuleBasedOnSequencePlayerInState=true, crossfadeDuration=0.3, logicType=TLT_Inertialization, priorityOrder=100}
   AnimStateTransitionNode_10 = AnimStateTransitionNode() {crossfadeDuration=0.35, logicType=TLT_Inertialization, priorityOrder=3}
   AnimStateTransitionNode_11 = AnimStateTransitionNode() {crossfadeDuration=0.3, logicType=TLT_Inertialization}
   AnimStateTransitionNode_12 = AnimStateTransitionNode() {blendMode=Linear, crossfadeDuration=0.3, priorityOrder=2}
   AnimStateTransitionNode_13 = AnimStateTransitionNode() {crossfadeDuration=0.4}
   AnimStateTransitionNode_14 = AnimStateTransitionNode() {logicType=TLT_Inertialization}
   AnimStateTransitionNode_15 = AnimStateTransitionNode() {bAutomaticRuleBasedOnSequencePlayerInState=true, crossfadeDuration=0.3, priorityOrder=2}
   AnimStateTransitionNode_16 = AnimStateTransitionNode()
   AnimStateTransitionNode_18 = AnimStateTransitionNode() {bAutomaticRuleBasedOnSequencePlayerInState=true, crossfadeDuration=0.4, priorityOrder=3}
   AnimStateTransitionNode_2 = AnimStateTransitionNode() {crossfadeDuration=0.6, logicType=TLT_Inertialization, priorityOrder=2}
   AnimStateTransitionNode_3 = AnimStateTransitionNode() {crossfadeDuration=0.4, logicType=TLT_Inertialization}
   AnimStateTransitionNode_4 = AnimStateTransitionNode() {crossfadeDuration=0.3, logicType=TLT_Inertialization, priorityOrder=100}
   AnimStateTransitionNode_5 = AnimStateTransitionNode() {logicType=TLT_Inertialization}
   AnimStateTransitionNode_6 = AnimStateTransitionNode() {logicType=TLT_Inertialization, priorityOrder=100}
   AnimStateTransitionNode_7 = AnimStateTransitionNode() {bAutomaticRuleBasedOnSequencePlayerInState=true, crossfadeDuration=0.4, logicType=TLT_Inertialization}
   AnimStateTransitionNode_8 = AnimStateTransitionNode() {crossfadeDuration=0.15, logicType=TLT_Inertialization}
   AnimStateTransitionNode_9 = AnimStateTransitionNode() {logicType=TLT_Inertialization, priorityOrder=2}
  ]
  links [
   {
    dst=Turn
    rule=VehicleFullBodyOverride.AnimGraphNode_StateMachine_2.CombatUpperBodyAdditive.AnimStateTransitionNode_0.Transition
    src=Idle
    via=AnimStateTransitionNode_0
   }
   {
    dst=JumpLand
    rule=VehicleFullBodyOverride.AnimGraphNode_StateMachine_2.CombatUpperBodyAdditive.AnimStateTransitionNode_6.Transition
    src=JumpFall
    via=AnimStateTransitionNode_6
   }
   {
    dst=Moving
    rule=VehicleFullBodyOverride.AnimGraphNode_StateMachine_2.CombatUpperBodyAdditive.AnimStateTransitionNode_8.Transition
    src=JumpFall
    via=AnimStateTransitionNode_8
   }
   {
    dst=Idle
    rule=VehicleFullBodyOverride.AnimGraphNode_StateMachine_2.CombatUpperBodyAdditive.AnimStateTransitionNode_1.Transition
    src=JumpLand
    via=AnimStateTransitionNode_1
   }
   {
    dst=Moving
    rule=VehicleFullBodyOverride.AnimGraphNode_StateMachine_2.CombatUpperBodyAdditive.AnimStateTransitionNode_9.Transition
    src=JumpLand
    via=AnimStateTransitionNode_9
   }
   {
    dst=JumpFall
    rule=VehicleFullBodyOverride.AnimGraphNode_StateMachine_2.CombatUpperBodyAdditive.AnimStateTransitionNode_7.Transition
    src=JumpStart
    via=AnimStateTransitionNode_7
   }
   {
    dst=Stop
    rule=VehicleFullBodyOverride.AnimGraphNode_StateMachine_2.CombatUpperBodyAdditive.AnimStateTransitionNode_10.Transition
    src=Moving
    via=AnimStateTransitionNode_10
   }
   {
    dst=JumpFall
    rule=VehicleFullBodyOverride.AnimGraphNode_StateMachine_2.CombatUpperBodyAdditive.AnimStateTransitionNode_11.Transition
    src=SA_ToJumpFall
    via=AnimStateTransitionNode_11
   }
   {
    dst=JumpStart
    rule=VehicleFullBodyOverride.AnimGraphNode_StateMachine_2.CombatUpperBodyAdditive.AnimStateTransitionNode_4.Transition
    src=SA_ToJumpStart
    via=AnimStateTransitionNode_4
   }
   {
    dst=Moving
    rule=VehicleFullBodyOverride.AnimGraphNode_StateMachine_2.CombatUpperBodyAdditive.AnimStateTransitionNode_2.Transition
    src=SA_ToMoving
    via=AnimStateTransitionNode_2
   }
   {
    dst=Swimming
    rule=VehicleFullBodyOverride.AnimGraphNode_StateMachine_2.CombatUpperBodyAdditive.AnimStateTransitionNode_3.Transition
    src=SA_ToSwimming
    via=AnimStateTransitionNode_3
   }
   {
    dst=Idle
    rule=VehicleFullBodyOverride.AnimGraphNode_StateMachine_2.CombatUpperBodyAdditive.AnimStateTransitionNode_18.Transition
    src=Stop
    via=AnimStateTransitionNode_18
   }
   {
    dst=Moving
    rule=VehicleFullBodyOverride.AnimGraphNode_StateMachine_2.CombatUpperBodyAdditive.AnimStateTransitionNode_12.Transition
    src=Stop
    via=AnimStateTransitionNode_12
   }
   {
    dst=Turn
    rule=VehicleFullBodyOverride.AnimGraphNode_StateMachine_2.CombatUpperBodyAdditive.AnimStateTransitionNode_13.Transition
    src=Stop
    via=AnimStateTransitionNode_13
   }
   {
    dst=Moving
    rule=VehicleFullBodyOverride.AnimGraphNode_StateMachine_2.CombatUpperBodyAdditive.AnimStateTransitionNode_5.Transition
    src=Swimming
    via=AnimStateTransitionNode_5
   }
   {
    dst=TurnSettle
    rule=VehicleFullBodyOverride.AnimGraphNode_StateMachine_2.CombatUpperBodyAdditive.AnimStateTransitionNode_14.Transition
    src=Turn
    via=AnimStateTransitionNode_14
   }
   {
    dst=Idle
    rule=VehicleFullBodyOverride.AnimGraphNode_StateMachine_2.CombatUpperBodyAdditive.AnimStateTransitionNode_15.Transition
    src=TurnSettle
    via=AnimStateTransitionNode_15
   }
   {
    dst=Turn
    rule=VehicleFullBodyOverride.AnimGraphNode_StateMachine_2.CombatUpperBodyAdditive.AnimStateTransitionNode_16.Transition
    src=TurnSettle
    via=AnimStateTransitionNode_16
   }
  ]
 }
 Moving {
  path=VehicleFullBodyOverride.AnimGraphNode_StateMachine_2.CombatUpperBodyAdditive.AnimStateNode_0.Moving
  nodes [
   n1 = Blendspace Player 'HardSaddleWolf_Player_CombatMove_Additive_BS'!(X=Get WolfSpeed2D(), PlayRate:float=Get RidingMovePlayRate()) -> (Pose) {blendSpace=&/HardSaddleWolf_Player_CombatMove_Additive_BS, groupName=Locomotion, method=SyncGroup}
   n2 = "Sequence Player (additive) 'HardSaddleWolf_Player_BodyRoll_Center'"() {sequence=&/HardSaddleWolf_Player_BodyRoll_Center}
   n3 = Get Is Boosting()
   n4 = "Sequence Player (additive) 'HardSaddleWolf_Player_CombatWalk_S_Additive'"!(PlayRate:float=Get Backwards Play Rate()) -> (Pose) {groupName=Locomotion, method=SyncGroup, sequence=&/HardSaddleWolf_Player_CombatWalk_S_Additive}
   Output Animation Pose(Result=Blend Poses by bool(
     BlendPose_0=@n4
     BlendPose_1=Blend Poses by bool(BlendPose_0=@n2, BlendPose_1=@n1, BlendTime_0=0.25, BlendTime_1=0.25, bActiveValue=@n3) {blendType=Linear}
     BlendTime_0=0.3
     BlendTime_1=0.3
     bActiveValue=Get Is Moving Backwards()
    ) {blendType=Linear})
  ]
 }
 Turn {
  path=VehicleFullBodyOverride.AnimGraphNode_StateMachine_2.CombatUpperBodyAdditive.AnimStateNode_1.Turn
  nodes [
   n1 = float / float(A=Get Turn in Place Rotation Speed() -> (Turn in Place Rotation Speed:double), B=168.75)
   n2 = Get Turn in Place Angle Delta() -> (Turn in Place Angle Delta:double)
   Output Animation Pose(Result=Blend Poses by bool(
     BlendPose_0="Sequence Player (additive) 'HardSaddleWolf_Player_TurnRight_Combat'"!(PlayRate:float=@n1) -> (Pose) {bLoopAnimation=false, sequence=&/HardSaddleWolf_Player_TurnRight_Combat}
     BlendPose_1="Sequence Player (additive) 'HardSaddleWolf_Player_TurnLeft_Combat'"!(PlayRate:float=@n1) -> (Pose) {bLoopAnimation=false, sequence=&/HardSaddleWolf_Player_TurnLeft_Combat}
     bActiveValue="float > float"(A=@n2)
    ) {blendType=Linear})
  ]
 }
 Swimming {
  path=VehicleFullBodyOverride.AnimGraphNode_StateMachine_2.CombatUpperBodyAdditive.AnimStateNode_2.Swimming
  nodes [
   n1 = "Sequence Player (additive) 'HardSaddle_Player_Combat_Swim_N_Additive'"() {groupName=Locomotion, method=SyncGroup, sequence=&/HardSaddle_Player_Combat_Swim_N_Additive}
   n2 = "Sequence Player (additive) 'HardSaddle_Player_Combat_Swim_S_Additive'"() {groupName=Locomotion, method=SyncGroup, sequence=&/HardSaddle_Player_Combat_Swim_S_Additive}
   n3 = Get Is Moving Backwards()
   Output Animation Pose(Result=Blend Poses by bool(
     BlendPose_0=Blend Poses by bool(BlendPose_0=@n2, BlendPose_1=@n1, BlendTime_0=0.35, BlendTime_1=0.35, bActiveValue=@n3) {blendType=Linear, transitionType=Inertialization}
     BlendPose_1="Sequence Player (additive) 'HardSaddle_Player_Combat_IdleSwim_Additive'"() {sequence=&/HardSaddle_Player_Combat_IdleSwim_Additive}
     BlendTime_0=0.5
     BlendTime_1=1
     bActiveValue=Get State Rule Swim Idle Swim Loco()
    ) {blendType=Linear})
  ]
 }
 TurnSettle {
  path=VehicleFullBodyOverride.AnimGraphNode_StateMachine_2.CombatUpperBodyAdditive.AnimStateNode_3.TurnSettle
  nodes [Output Animation Pose(Result=Blend Poses by bool(
     BlendPose_0="Sequence Player (additive) 'HardSaddleWolf_Player_TurnRight_Combat_Settle'"() {bLoopAnimation=false, sequence=&/HardSaddleWolf_Player_TurnRight_Combat_Settle}
     BlendPose_1="Sequence Player (additive) 'HardSaddleWolf_Player_TurnLeft_Combat_Settle'"() {bLoopAnimation=false, sequence=&/HardSaddleWolf_Player_TurnLeft_Combat_Settle}
     BlendTime_0=0.25
     BlendTime_1=0.25
     bActiveValue=Get Was Turning Right()
    ) {blendType=Linear})]
 }
 Idle {
  path=VehicleFullBodyOverride.AnimGraphNode_StateMachine_2.CombatUpperBodyAdditive.AnimStateNode_4.Idle
  nodes [Output Animation Pose(Result="Sequence Player (additive) 'HardSaddle_Player_Combat_Idle_Additive'"() {sequence=&/HardSaddle_Player_Combat_Idle_Additive})]
 }
 JumpStart {
  path=VehicleFullBodyOverride.AnimGraphNode_StateMachine_2.CombatUpperBodyAdditive.AnimStateNode_5.JumpStart
  nodes [Output Animation Pose(Result="Sequence Player (additive) 'HardSaddleWolf_Player_CombatJumpStart_Add'"() {bLoopAnimation=false, sequence=&/HardSaddleWolf_Player_CombatJumpStart_Add})]
 }
 JumpFall {
  path=VehicleFullBodyOverride.AnimGraphNode_StateMachine_2.CombatUpperBodyAdditive.AnimStateNode_6.JumpFall
  nodes [Output Animation Pose(Result=Blend Poses by bool(
     BlendPose_0="Sequence Player (additive) 'HardSaddleWolf_plyr_SlopeSliding_AddUpper'"() {sequence=/RidingContent/Animation/HardSaddle/Wolf/SlopeSliding/HardSaddleWolf_plyr_SlopeSliding_AddUpper}
     BlendPose_1="Sequence Player (additive) 'HardSaddleWolf_Player_CombatJumpLoop_Add'"() {sequence=&/HardSaddleWolf_Player_CombatJumpLoop_Add}
     bActiveValue=Get bIsSlopeSliding()
    ) {blendType=Linear})]
 }
 Stop {
  path=VehicleFullBodyOverride.AnimGraphNode_StateMachine_2.CombatUpperBodyAdditive.AnimStateNode_7.Stop
  nodes [
   AnimGraphNode_StateMachine_0 = AdditiveStopTransition() {is=title AnimGraphNode_StateMachine, bAllowConduitEntryStates=true, bSkipFirstUpdateTransition=false, maxTransitionsPerFrame=1}
   Output Animation Pose(Result=@AnimGraphNode_StateMachine_0)
  ]
 }
 AdditiveStopTransition {
  entry=Conduit
  path=VehicleFullBodyOverride.AnimGraphNode_StateMachine_2.CombatUpperBodyAdditive.AnimStateNode_7.Stop.AnimGraphNode_StateMachine_0.AdditiveStopTransition
  ids=id
  nodes [
   AnimStateConduitNode_0 = Conduit()
   AnimStateNode_0 = FeetInAir()
   AnimStateNode_1 = FrontFeetPlanted()
   AnimStateNode_2 = BackFeetPlanted()
   AnimStateNode_3 = Backwards()
   AnimStateTransitionNode_1 = AnimStateTransitionNode() {logicType=TLT_Inertialization, priorityOrder=0}
   AnimStateTransitionNode_3 = AnimStateTransitionNode() {logicType=TLT_Inertialization}
   AnimStateTransitionNode_4 = AnimStateTransitionNode() {logicType=TLT_Inertialization, priorityOrder=2}
   AnimStateTransitionNode_5 = AnimStateTransitionNode() {logicType=TLT_Inertialization, priorityOrder=100}
  ]
  links [
   {
    dst=BackFeetPlanted
    rule=VehicleFullBodyOverride.AnimGraphNode_StateMachine_2.CombatUpperBodyAdditive.AnimStateNode_7.Stop.AnimGraphNode_StateMachine_0.AdditiveStopTransition.AnimStateTransitionNode_5.Transition
    src=Conduit
    via=AnimStateTransitionNode_5
   }
   {
    dst=Backwards
    rule=VehicleFullBodyOverride.AnimGraphNode_StateMachine_2.CombatUpperBodyAdditive.AnimStateNode_7.Stop.AnimGraphNode_StateMachine_0.AdditiveStopTransition.AnimStateTransitionNode_1.Transition
    src=Conduit
    via=AnimStateTransitionNode_1
   }
   {
    dst=FeetInAir
    rule=VehicleFullBodyOverride.AnimGraphNode_StateMachine_2.CombatUpperBodyAdditive.AnimStateNode_7.Stop.AnimGraphNode_StateMachine_0.AdditiveStopTransition.AnimStateTransitionNode_3.Transition
    src=Conduit
    via=AnimStateTransitionNode_3
   }
   {
    dst=FrontFeetPlanted
    rule=VehicleFullBodyOverride.AnimGraphNode_StateMachine_2.CombatUpperBodyAdditive.AnimStateNode_7.Stop.AnimGraphNode_StateMachine_0.AdditiveStopTransition.AnimStateTransitionNode_4.Transition
    src=Conduit
    via=AnimStateTransitionNode_4
   }
  ]
 }
 Conduit {
  path=VehicleFullBodyOverride.AnimGraphNode_StateMachine_2.CombatUpperBodyAdditive.AnimStateNode_7.Stop.AnimGraphNode_StateMachine_0.AdditiveStopTransition.AnimStateConduitNode_0.Conduit
  nodes [Result(bCanEnterTransition=true)]
 }
 FeetInAir {
  path=VehicleFullBodyOverride.AnimGraphNode_StateMachine_2.CombatUpperBodyAdditive.AnimStateNode_7.Stop.AnimGraphNode_StateMachine_0.AdditiveStopTransition.AnimStateNode_0.FeetInAir
  nodes [Output Animation Pose(Result="Sequence Player (additive) 'HardSaddleWolf_Player_Stop_FeetInAir_Combat_Additive'"() {bLoopAnimation=false, sequence=&/HardSaddleWolf_Player_Stop_FeetInAir_Combat_Additive})]
 }
 FrontFeetPlanted {
  path=VehicleFullBodyOverride.AnimGraphNode_StateMachine_2.CombatUpperBodyAdditive.AnimStateNode_7.Stop.AnimGraphNode_StateMachine_0.AdditiveStopTransition.AnimStateNode_1.FrontFeetPlanted
  nodes [Output Animation Pose(Result="Sequence Player (additive) 'HardSaddleWolf_Player_Stop_FrontFeetPlanted_Combat_Additive'"() {bLoopAnimation=false, sequence=&/HardSaddleWolf_Player_Stop_FrontFeetPlanted_Combat_Additive})]
 }
 BackFeetPlanted {
  path=VehicleFullBodyOverride.AnimGraphNode_StateMachine_2.CombatUpperBodyAdditive.AnimStateNode_7.Stop.AnimGraphNode_StateMachine_0.AdditiveStopTransition.AnimStateNode_2.BackFeetPlanted
  nodes [Output Animation Pose(Result="Sequence Player (additive) 'HardSaddleWolf_Player_Stop_BackFeetPlanted_Combat_Additive'"() {bLoopAnimation=false, sequence=&/HardSaddleWolf_Player_Stop_BackFeetPlanted_Combat_Additive})]
 }
 Backwards {
  path=VehicleFullBodyOverride.AnimGraphNode_StateMachine_2.CombatUpperBodyAdditive.AnimStateNode_7.Stop.AnimGraphNode_StateMachine_0.AdditiveStopTransition.AnimStateNode_3.Backwards
  nodes [
   AnimGraphNode_StateMachine_0 = AdditiveStopBackwards() {is=title AnimGraphNode_StateMachine, bAllowConduitEntryStates=true, bSkipFirstUpdateTransition=false, maxTransitionsPerFrame=1}
   Output Animation Pose(Result=@AnimGraphNode_StateMachine_0)
  ]
 }
 AdditiveStopBackwards {
  entry=Conduit
  path=VehicleFullBodyOverride.AnimGraphNode_StateMachine_2.CombatUpperBodyAdditive.AnimStateNode_7.Stop.AnimGraphNode_StateMachine_0.AdditiveStopTransition.AnimStateNode_3.Backwards.AnimGraphNode_StateMachine_0.AdditiveStopBackwards
  ids=id
  nodes [
   AnimStateConduitNode_0 = Conduit()
   AnimStateNode_0 = LeftBackFeetForward()
   AnimStateNode_1 = RightBackFeetForward()
   AnimStateNode_2 = LeftPlantedRightPass()
   AnimStateNode_3 = RightPlantedLeftPass()
   AnimStateTransitionNode_0 = AnimStateTransitionNode() {logicType=TLT_Inertialization}
   AnimStateTransitionNode_1 = AnimStateTransitionNode() {logicType=TLT_Inertialization, priorityOrder=2}
   AnimStateTransitionNode_2 = AnimStateTransitionNode() {logicType=TLT_Inertialization, priorityOrder=3}
   AnimStateTransitionNode_3 = AnimStateTransitionNode() {logicType=TLT_Inertialization, priorityOrder=100}
  ]
  links [
   {
    dst=LeftBackFeetForward
    rule=VehicleFullBodyOverride.AnimGraphNode_StateMachine_2.CombatUpperBodyAdditive.AnimStateNode_7.Stop.AnimGraphNode_StateMachine_0.AdditiveStopTransition.AnimStateNode_3.Backwards.AnimGraphNode_StateMachine_0.AdditiveStopBackwards.AnimStateTransitionNode_0.AnimationTransitionGraph_1
    src=Conduit
    via=AnimStateTransitionNode_0
   }
   {
    dst=LeftPlantedRightPass
    rule=VehicleFullBodyOverride.AnimGraphNode_StateMachine_2.CombatUpperBodyAdditive.AnimStateNode_7.Stop.AnimGraphNode_StateMachine_0.AdditiveStopTransition.AnimStateNode_3.Backwards.AnimGraphNode_StateMachine_0.AdditiveStopBackwards.AnimStateTransitionNode_2.AnimationTransitionGraph_1
    src=Conduit
    via=AnimStateTransitionNode_2
   }
   {
    dst=RightBackFeetForward
    rule=VehicleFullBodyOverride.AnimGraphNode_StateMachine_2.CombatUpperBodyAdditive.AnimStateNode_7.Stop.AnimGraphNode_StateMachine_0.AdditiveStopTransition.AnimStateNode_3.Backwards.AnimGraphNode_StateMachine_0.AdditiveStopBackwards.AnimStateTransitionNode_1.AnimationTransitionGraph_1
    src=Conduit
    via=AnimStateTransitionNode_1
   }
   {
    dst=RightPlantedLeftPass
    rule=VehicleFullBodyOverride.AnimGraphNode_StateMachine_2.CombatUpperBodyAdditive.AnimStateNode_7.Stop.AnimGraphNode_StateMachine_0.AdditiveStopTransition.AnimStateNode_3.Backwards.AnimGraphNode_StateMachine_0.AdditiveStopBackwards.AnimStateTransitionNode_3.AnimationTransitionGraph_1
    src=Conduit
    via=AnimStateTransitionNode_3
   }
  ]
 }
 Conduit {
  path=VehicleFullBodyOverride.AnimGraphNode_StateMachine_2.CombatUpperBodyAdditive.AnimStateNode_7.Stop.AnimGraphNode_StateMachine_0.AdditiveStopTransition.AnimStateNode_3.Backwards.AnimGraphNode_StateMachine_0.AdditiveStopBackwards.AnimStateConduitNode_0.Conduit
  nodes [Result(bCanEnterTransition=true)]
 }
 LeftBackFeetForward {
  path=VehicleFullBodyOverride.AnimGraphNode_StateMachine_2.CombatUpperBodyAdditive.AnimStateNode_7.Stop.AnimGraphNode_StateMachine_0.AdditiveStopTransition.AnimStateNode_3.Backwards.AnimGraphNode_StateMachine_0.AdditiveStopBackwards.AnimStateNode_0.LeftBackFeetForward
  nodes [Output Animation Pose(Result="Sequence Player (additive) 'HardSaddleWolf_Player_Stop_Backwards_Combat_Add'"() {bLoopAnimation=false, sequence=&/HardSaddleWolf_Player_Stop_Backwards_Combat_Add})]
 }
 RightBackFeetForward {
  path=VehicleFullBodyOverride.AnimGraphNode_StateMachine_2.CombatUpperBodyAdditive.AnimStateNode_7.Stop.AnimGraphNode_StateMachine_0.AdditiveStopTransition.AnimStateNode_3.Backwards.AnimGraphNode_StateMachine_0.AdditiveStopBackwards.AnimStateNode_1.RightBackFeetForward
  nodes [Output Animation Pose(Result="Sequence Player (additive) 'HardSaddleWolf_Player_Stop_Backwards_Combat_Add_fr19'"() {sequence=&/HardSaddleWolf_Player_Stop_Backwards_Combat_Add_fr19})]
 }
 LeftPlantedRightPass {
  path=VehicleFullBodyOverride.AnimGraphNode_StateMachine_2.CombatUpperBodyAdditive.AnimStateNode_7.Stop.AnimGraphNode_StateMachine_0.AdditiveStopTransition.AnimStateNode_3.Backwards.AnimGraphNode_StateMachine_0.AdditiveStopBackwards.AnimStateNode_2.LeftPlantedRightPass
  nodes [Output Animation Pose(Result="Sequence Player (additive) 'HardSaddleWolf_Player_Stop_Backwards_Combat_Add_fr24'"() {bLoopAnimation=false, sequence=&/HardSaddleWolf_Player_Stop_Backwards_Combat_Add_fr24})]
 }
 RightPlantedLeftPass {
  path=VehicleFullBodyOverride.AnimGraphNode_StateMachine_2.CombatUpperBodyAdditive.AnimStateNode_7.Stop.AnimGraphNode_StateMachine_0.AdditiveStopTransition.AnimStateNode_3.Backwards.AnimGraphNode_StateMachine_0.AdditiveStopBackwards.AnimStateNode_3.RightPlantedLeftPass
  nodes [Output Animation Pose(Result="Sequence Player (additive) 'HardSaddleWolf_Player_Stop_Backwards_Combat_Add_fr7'"() {bLoopAnimation=false, sequence=&/HardSaddleWolf_Player_Stop_Backwards_Combat_Add_fr7})]
 }
 AnimationTransitionGraph_1 {
  path=VehicleFullBodyOverride.AnimGraphNode_StateMachine_2.CombatUpperBodyAdditive.AnimStateNode_7.Stop.AnimGraphNode_StateMachine_0.AdditiveStopTransition.AnimStateNode_3.Backwards.AnimGraphNode_StateMachine_0.AdditiveStopBackwards.AnimStateTransitionNode_0.AnimationTransitionGraph_1
  nodes [Result()]
 }
 AnimationTransitionGraph_1 {
  path=VehicleFullBodyOverride.AnimGraphNode_StateMachine_2.CombatUpperBodyAdditive.AnimStateNode_7.Stop.AnimGraphNode_StateMachine_0.AdditiveStopTransition.AnimStateNode_3.Backwards.AnimGraphNode_StateMachine_0.AdditiveStopBackwards.AnimStateTransitionNode_1.AnimationTransitionGraph_1
  nodes [Result()]
 }
 AnimationTransitionGraph_1 {
  path=VehicleFullBodyOverride.AnimGraphNode_StateMachine_2.CombatUpperBodyAdditive.AnimStateNode_7.Stop.AnimGraphNode_StateMachine_0.AdditiveStopTransition.AnimStateNode_3.Backwards.AnimGraphNode_StateMachine_0.AdditiveStopBackwards.AnimStateTransitionNode_2.AnimationTransitionGraph_1
  nodes [Result()]
 }
 AnimationTransitionGraph_1 {
  path=VehicleFullBodyOverride.AnimGraphNode_StateMachine_2.CombatUpperBodyAdditive.AnimStateNode_7.Stop.AnimGraphNode_StateMachine_0.AdditiveStopTransition.AnimStateNode_3.Backwards.AnimGraphNode_StateMachine_0.AdditiveStopBackwards.AnimStateTransitionNode_3.AnimationTransitionGraph_1
  nodes [Result(bCanEnterTransition=true)]
 }
 Transition {
  path=VehicleFullBodyOverride.AnimGraphNode_StateMachine_2.CombatUpperBodyAdditive.AnimStateNode_7.Stop.AnimGraphNode_StateMachine_0.AdditiveStopTransition.AnimStateTransitionNode_1.Transition
  nodes [Result(bCanEnterTransition=Get Should Play Stop Backwards())]
 }
 Transition {
  path=VehicleFullBodyOverride.AnimGraphNode_StateMachine_2.CombatUpperBodyAdditive.AnimStateNode_7.Stop.AnimGraphNode_StateMachine_0.AdditiveStopTransition.AnimStateTransitionNode_3.Transition
  nodes [Result()]
 }
 Transition {
  path=VehicleFullBodyOverride.AnimGraphNode_StateMachine_2.CombatUpperBodyAdditive.AnimStateNode_7.Stop.AnimGraphNode_StateMachine_0.AdditiveStopTransition.AnimStateTransitionNode_4.Transition
  nodes [Result()]
 }
 Transition {
  path=VehicleFullBodyOverride.AnimGraphNode_StateMachine_2.CombatUpperBodyAdditive.AnimStateNode_7.Stop.AnimGraphNode_StateMachine_0.AdditiveStopTransition.AnimStateTransitionNode_5.Transition
  nodes [Result(bCanEnterTransition=true)]
 }
 JumpLand {
  path=VehicleFullBodyOverride.AnimGraphNode_StateMachine_2.CombatUpperBodyAdditive.AnimStateNode_8.JumpLand
  nodes [Output Animation Pose(Result="Sequence Player (additive) 'HardSaddleWolf_Player_CombatJumpLand_Add'"() {bLoopAnimation=false, sequence=&/HardSaddleWolf_Player_CombatJumpLand_Add})]
 }
 Transition {
  path=VehicleFullBodyOverride.AnimGraphNode_StateMachine_2.CombatUpperBodyAdditive.AnimStateTransitionNode_0.Transition
  nodes [Result(bCanEnterTransition=Get Is Turning in Place())]
 }
 Transition {path=VehicleFullBodyOverride.AnimGraphNode_StateMachine_2.CombatUpperBodyAdditive.AnimStateTransitionNode_1.Transition, nodes [Result()]}
 Transition {
  path=VehicleFullBodyOverride.AnimGraphNode_StateMachine_2.CombatUpperBodyAdditive.AnimStateTransitionNode_10.Transition
  nodes [Result(bCanEnterTransition=Get State Rule Moving Idle())]
 }
 Transition {
  path=VehicleFullBodyOverride.AnimGraphNode_StateMachine_2.CombatUpperBodyAdditive.AnimStateTransitionNode_11.Transition
  nodes [Result(bCanEnterTransition=Get Is Falling())]
 }
 Transition {
  path=VehicleFullBodyOverride.AnimGraphNode_StateMachine_2.CombatUpperBodyAdditive.AnimStateTransitionNode_12.Transition
  nodes [Result(bCanEnterTransition=Get State Rule Idle Moving())]
 }
 Transition {
  path=VehicleFullBodyOverride.AnimGraphNode_StateMachine_2.CombatUpperBodyAdditive.AnimStateTransitionNode_13.Transition
  nodes [Result(bCanEnterTransition=Get Is Turning in Place())]
 }
 Transition {
  path=VehicleFullBodyOverride.AnimGraphNode_StateMachine_2.CombatUpperBodyAdditive.AnimStateTransitionNode_14.Transition
  nodes [Result(bCanEnterTransition=NOT Boolean(A=Get Is Turning in Place()))]
 }
 Transition {path=VehicleFullBodyOverride.AnimGraphNode_StateMachine_2.CombatUpperBodyAdditive.AnimStateTransitionNode_15.Transition, nodes [Result()]}
 Transition {
  path=VehicleFullBodyOverride.AnimGraphNode_StateMachine_2.CombatUpperBodyAdditive.AnimStateTransitionNode_16.Transition
  nodes [Result(bCanEnterTransition=Get Is Turning in Place())]
 }
 Transition {path=VehicleFullBodyOverride.AnimGraphNode_StateMachine_2.CombatUpperBodyAdditive.AnimStateTransitionNode_18.Transition, nodes [Result()]}
 Transition {
  path=VehicleFullBodyOverride.AnimGraphNode_StateMachine_2.CombatUpperBodyAdditive.AnimStateTransitionNode_2.Transition
  nodes [Result(bCanEnterTransition=Get State Rule Idle Moving())]
 }
 Transition {
  path=VehicleFullBodyOverride.AnimGraphNode_StateMachine_2.CombatUpperBodyAdditive.AnimStateTransitionNode_3.Transition
  nodes [Result(bCanEnterTransition=Get Is Surface Swimming())]
 }
 Transition {
  path=VehicleFullBodyOverride.AnimGraphNode_StateMachine_2.CombatUpperBodyAdditive.AnimStateTransitionNode_4.Transition
  nodes [Result(bCanEnterTransition=Get Is Jumping())]
 }
 Transition {
  path=VehicleFullBodyOverride.AnimGraphNode_StateMachine_2.CombatUpperBodyAdditive.AnimStateTransitionNode_5.Transition
  nodes [Result(bCanEnterTransition=NOT Boolean(A=Get Is Surface Swimming()))]
 }
 Transition {
  path=VehicleFullBodyOverride.AnimGraphNode_StateMachine_2.CombatUpperBodyAdditive.AnimStateTransitionNode_6.Transition
  nodes [Result(bCanEnterTransition=Get State Rule Fall Idle())]
 }
 Transition {path=VehicleFullBodyOverride.AnimGraphNode_StateMachine_2.CombatUpperBodyAdditive.AnimStateTransitionNode_7.Transition, nodes [Result()]}
 Transition {
  path=VehicleFullBodyOverride.AnimGraphNode_StateMachine_2.CombatUpperBodyAdditive.AnimStateTransitionNode_8.Transition
  nodes [Result(bCanEnterTransition=Get State Rule Fall Moving())]
 }
 Transition {
  path=VehicleFullBodyOverride.AnimGraphNode_StateMachine_2.CombatUpperBodyAdditive.AnimStateTransitionNode_9.Transition
  nodes [Result(bCanEnterTransition=Get State Rule Jump Land Moving())]
 }
 MeleeOffset {
  entry=MeleeUpperBodyOffset
  path=VehicleFullBodyOverride.AnimGraphNode_StateMachine_3.MeleeOffset
  ids=id
  nodes [AnimStateNode_0 = MeleeUpperBodyOffset() {is=title AnimStateNode}]
 }
 MeleeUpperBodyOffset {
  path=VehicleFullBodyOverride.AnimGraphNode_StateMachine_3.MeleeOffset.AnimStateNode_0.MeleeUpperBodyOffset
  nodes [
   n1 = Get BodyYaw() -> (BodyYaw:float)
   n2 = Blend Poses by bool(
    BlendPose_0=Blendspace Player 'HardSaddleWolf_MeleeOffset_1H2Swing_BS'!(X=@n1) -> (Pose) {blendSpace=/RidingContent/Animation/HardSaddle/Wolf/MeleeOffset/HardSaddleWolf_MeleeOffset_1H2Swing_BS}
    BlendPose_1=Blend Poses by bool(
     BlendPose_0=Blendspace Player 'HardSaddleWolf_MeleeOffset_2HSword_BS'!(X=@n1) -> (Pose) {blendSpace=/RidingContent/Animation/HardSaddle/Wolf/MeleeOffset/HardSaddleWolf_MeleeOffset_2HSword_BS}
     BlendPose_1=Additive Identity Pose()
     bActiveValue=Get bIsTwoHandedSword() -> (bIsTwoHandedSword:bool)
    ) {blendType=Linear}
    bActiveValue=Get bIsOneHandedDualWield() -> (bIsOneHandedDualWield:bool)
   ) {blendType=Linear}
   n3 = Blendspace Player 'HardSaddleWolf_MeleeOffset_DW_BS'!(X=@n1) -> (Pose) {blendSpace=/RidingContent/Animation/HardSaddle/Wolf/MeleeOffset/HardSaddleWolf_MeleeOffset_DW_BS}
   n4 = Get bIsDualWield() -> (bIsDualWield:bool)
   n5 = Get BodyYaw()
   Output Animation Pose(Result=Apply Additive(
     Base=Blend Poses by bool(BlendPose_0=@n3, BlendPose_1=@n2, bActiveValue=@n4) {blendType=Linear}
     Additive=Blendspace Player 'HardSaddleBoar_Player_MeleeLegCorrection_BS'!(X=@n5) -> (Pose) {blendSpace=/RidingContent/Animation/HardSaddle/Boar/HardSaddleBoar_Player_MeleeLegCorrection_BS}
    ) {alphaBoolBlend {blendOption=Linear}})
  ]
 }
 EntryAdditive {
  entry=AddPose
  path=VehicleFullBodyOverride.AnimGraphNode_StateMachine_4.EntryAdditive
  ids=id
  nodes [
   AnimStateNode_0 = AddPose() {is=title AnimStateNode}
   AnimStateNode_1 = EntryAdd() {is=title AnimStateNode}
   AnimStateTransitionNode_0 = AnimStateTransitionNode() {crossfadeDuration=0.1, logicType=TLT_Inertialization}
   AnimStateTransitionNode_1 = AnimStateTransitionNode() {bAutomaticRuleBasedOnSequencePlayerInState=true, crossfadeDuration=0.3, logicType=TLT_Inertialization}
  ]
  links [
   {
    dst=EntryAdd
    rule=VehicleFullBodyOverride.AnimGraphNode_StateMachine_4.EntryAdditive.AnimStateTransitionNode_0.Transition
    src=AddPose
    via=AnimStateTransitionNode_0
   }
   {
    dst=AddPose
    rule=VehicleFullBodyOverride.AnimGraphNode_StateMachine_4.EntryAdditive.AnimStateTransitionNode_1.Transition
    src=EntryAdd
    via=AnimStateTransitionNode_1
   }
  ]
 }
 AddPose {
  path=VehicleFullBodyOverride.AnimGraphNode_StateMachine_4.EntryAdditive.AnimStateNode_0.AddPose
  nodes [Output Animation Pose(Result=Additive Identity Pose())]
 }
 EntryAdd {
  path=VehicleFullBodyOverride.AnimGraphNode_StateMachine_4.EntryAdditive.AnimStateNode_1.EntryAdd
  nodes [Output Animation Pose(Result="Sequence Player (additive) 'HardSaddleWolf_Player_SimpleEntry'"() {bLoopAnimation=false, sequence=&/HardSaddleWolf_Player_SimpleEntry})]
 }
 Transition {
  path=VehicleFullBodyOverride.AnimGraphNode_StateMachine_4.EntryAdditive.AnimStateTransitionNode_0.Transition
  nodes [Result(bCanEnterTransition=Get bJustEntered() -> (bJustEntered:bool))]
 }
 Transition {path=VehicleFullBodyOverride.AnimGraphNode_StateMachine_4.EntryAdditive.AnimStateTransitionNode_1.Transition, nodes [Result()]}
 VehicleLowerBodyOverride {nodes [
   Input Pose() {name=InPoseLowerBodyDefault}
   Input Pose() {name=InPoseUpperBody}
   Output Pose() {layerGroup=VehicleOverride}
  ]}
 VehicleSplitBodyOverride {nodes [
   Input Pose() {name=InPoseSplitBody}
   Input Pose() {name=InPoseUpperAndLowerBody}
   Input Pose() {name=InPoseUpperAndLowerBodyRemoveRoot}
   Output Pose() {layerGroup=Vehicle Override}
  ]}
]
variables [
 Backwards Play Rate {type=number}
 BodyRoll {type=number}
 Is Boosting {type=boolean}
 Is Falling {type=boolean}
 Is Jumping {type=boolean}
 Is Moving Backwards {type=boolean}
 Is Surface Swimming {type=boolean}
 Is Turning in Place {type=boolean}
 Play Land Additive {type=boolean}
 Should Play Stop Backwards {type=boolean}
 State Rule Fall Idle {type=boolean}
 State Rule Fall Moving {type=boolean}
 State Rule Idle Moving {type=boolean}
 State Rule Jump Land Moving {type=boolean}
 State Rule Moving Idle {type=boolean}
 State Rule Swim Idle Swim Loco {type=boolean}
 Turn in Place Angle Delta {type=number}
 Turn in Place Rotation Speed {type=number}
 Was Turning Right {type=boolean}
 WolfSpeed2D {type=number}
]
